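# Anim Blueprint for the TwilightSpotSpell pickaxe, adding secondary motion to its four left/right roll sets and handle paper.
# Two AnimDynamics chains blend on bIsInFrontEnd: world-space with linear springs and pitch offsets when true, angular springs only otherwise.
# SetStage01–04 events each unhide one material element and flag that roll active.
# Timed stage-advance logic for melee swings exists, but its entry events are unwired.
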AnimBlueprint /BRCosmetics/Weapons/FORT_Melee/Pickaxe_TwilightSpotSpell/Meshes/TwilightSpotSpell_Axe_AnimBP
target_skeleton=&/TwilightSpotSpell_Axe_Skeleton
graphs [
 AnimGraph {nodes [
   AnimGraphNode_SubInput_0 = Input Pose()
   n1 = "Transform (Modify) Bone - Bone: dyn_roll_MSTR_l"!(
    ComponentPose=Local To Component(LocalPose=@AnimGraphNode_SubInput_0)
    bAlphaBoolEnabled:bool=True
    Translation="0,0,0"
    Rotation="-35,0,0"
    Scale="1,1,1"
   ) -> (Pose) {
    alphaInputType=Bool
    rotationMode=BMM_Additive
    rotationSpace=BCS_ParentBoneSpace
    alphaBoolBlend {blendOption=Linear}
    boneToModify {boneName=dyn_roll_MSTR_l}
    rotation {pitch=-35}
   }
   n2 = Anim Dynamics - Bone: dyn_roll_01_l(ComponentPose=@n1, Alpha=0, ExternalForce="0,0,0") {
    angularDampingOverride=0.9999
    angularSpringConstant=1000
    bAngularSpring=true
    bLinearSpring=true
    bOverrideAngularDamping=true
    bOverrideLinearDamping=true
    bUseGravityOverride=true
    linearDampingOverride=0.9999
    linearSpringConstant=10
    simulationSpace=World
    alphaBoolBlend {blendOption=Linear}
    boundBone {boneName=dyn_roll_01_l}
    componentAppliedLinearAccClamp=<2000 2000 1000>
    componentLinearAccScale=<1 1 1>
    componentLinearVelScale=<1 1 1>
    gravityOverride {z=-980}
    physicsBodyDefinitions [{
      boundBone {boneName=dyn_roll_01_l}
      boxExtents=<22 22 22>
      constraintSetup {
       twistAxis=AxisY
       angularLimitsMax=<0.1 50 0.1>
       angularLimitsMin=<-0.1 -35 -0.1>
       angularTarget=<1 0 0>
       linearAxesMax=<0.1 0.1 0.1>
       linearAxesMin=<-0.1 -0.1 -0.1>
      }
      localJointOffset=<2.5 0 0>
     }]
   }
   n3 = Anim Dynamics - Bone: dyn_roll_02_l(ComponentPose=@n2, Alpha=1, ExternalForce="0,0,0") {
    angularDampingOverride=0.9999
    angularSpringConstant=500
    bAngularSpring=true
    bLinearSpring=true
    bOverrideAngularDamping=true
    bOverrideLinearDamping=true
    bUseGravityOverride=true
    linearDampingOverride=0.9999
    linearSpringConstant=10
    simulationSpace=World
    alphaBoolBlend {blendOption=Linear}
    boundBone {boneName=dyn_roll_02_l}
    componentAppliedLinearAccClamp=<2000 2000 2000>
    componentLinearAccScale=<1 1 1>
    componentLinearVelScale=<1 1 1>
    gravityOverride {z=-900}
    physicsBodyDefinitions [{
      boundBone {boneName=dyn_roll_02_l}
      boxExtents=<17 17 17>
      constraintSetup {
       angularLimitsMax=<0.1 20 0.1>
       angularLimitsMin=<-0.1 -10 -0.1>
       angularTarget=<1 0 0>
       linearAxesMax=<0.1 0.1 0.1>
       linearAxesMin=<-0.1 -0.1 -0.1>
      }
      localJointOffset=<2 0 0>
     }]
   }
   n4 = "Transform (Modify) Bone - Bone: dyn_roll_MSTR_r"(ComponentPose=@n3, Alpha=1, Translation="0,0,0", Rotation="25,0,0", Scale="1,1,1") {rotationMode=BMM_Additive, alphaBoolBlend {blendOption=Linear}, boneToModify {boneName=dyn_roll_MSTR_r}, rotation {pitch=25}}
   n5 = "Transform (Modify) Bone - Bone: dyn_roll_02_r"(ComponentPose=@n4, Alpha=1, Translation="0,0,0", Rotation="20,0,0", Scale="1,1,1") {rotationMode=BMM_Additive, alphaBoolBlend {blendOption=Linear}, boneToModify {boneName=dyn_roll_02_r}, rotation {pitch=20}}
   n6 = Anim Dynamics - Bone: dyn_roll_01_r(ComponentPose=@n5, Alpha=0.4, ExternalForce="0,0,0") {
    angularBiasOverride=0.9999
    angularDampingOverride=0.9999
    angularSpringConstant=1000
    bAngularSpring=true
    bLinearSpring=true
    bOverrideAngularBias=true
    bOverrideAngularDamping=true
    bOverrideLinearDamping=true
    bUseGravityOverride=true
    linearDampingOverride=1
    linearSpringConstant=10
    simulationSpace=World
    alphaBoolBlend {blendOption=Linear}
    boundBone {boneName=dyn_roll_01_r}
    componentAppliedLinearAccClamp=<20000 20000 1000>
    componentLinearAccScale=<1 1 1>
    componentLinearVelScale=<1 1 1>
    gravityOverride {z=-980}
    physicsBodyDefinitions [{
      boundBone {boneName=dyn_roll_01_r}
      boxExtents=<17 17 17>
      constraintSetup {
       angularLimitsMax=<0.1 30 0.1>
       angularLimitsMin=<-0.1 -30 -0.1>
       angularTarget=<1 0 0>
       linearAxesMax=<0.1 0.1 0.1>
       linearAxesMin=<-0.1 -0.1 -0.1>
      }
      localJointOffset=<2 0 0>
     }]
   }
   n7 = Anim Dynamics - Bone: dyn_roll_02_r(ComponentPose=@n6, Alpha=0.4, ExternalForce="0,0,0") {
    angularBiasOverride=0.9999
    angularDampingOverride=0.9999
    angularSpringConstant=500
    bAngularSpring=true
    bLinearSpring=true
    bOverrideAngularBias=true
    bOverrideAngularDamping=true
    bOverrideLinearDamping=true
    bUseGravityOverride=true
    linearDampingOverride=1
    linearSpringConstant=10
    simulationSpace=World
    alphaBoolBlend {blendOption=Linear}
    boundBone {boneName=dyn_roll_02_r}
    componentAppliedLinearAccClamp=<2000 2000 2000>
    componentLinearAccScale=<1 1 1>
    componentLinearVelScale=<1 1 1>
    gravityOverride {z=-900}
    physicsBodyDefinitions [{
      boundBone {boneName=dyn_roll_02_r}
      boxExtents=<17 17 17>
      constraintSetup {
       angularLimitsMax=<0.1 25 0.1>
       angularLimitsMin=<-0.1 -30 -0.1>
       angularTarget=<1 0 0>
       linearAxesMax=<0.1 0.1 0.1>
       linearAxesMin=<-0.1 -0.1 -0.1>
      }
      localJointOffset=<1.5 0 0>
     }]
   }
   n8 = "Transform (Modify) Bone - Bone: dyn_roll3_MSTR_l"!(ComponentPose=@n7, bAlphaBoolEnabled:bool=True, Translation="0,0,0", Rotation="-35,0,0", Scale="1,1,1") -> (Pose) {
    alphaInputType=Bool
    rotationMode=BMM_Additive
    rotationSpace=BCS_ParentBoneSpace
    alphaBoolBlend {blendOption=Linear}
    boneToModify {boneName=dyn_roll3_MSTR_l}
    rotation {pitch=-35}
   }
   n9 = Anim Dynamics - Bone: dyn_roll2_01_l(ComponentPose=@n8, Alpha=1, ExternalForce="0,0,0") {
    angularDampingOverride=0.9999
    angularSpringConstant=1000
    bAngularSpring=true
    bLinearSpring=true
    bOverrideAngularDamping=true
    bOverrideLinearDamping=true
    bUseGravityOverride=true
    linearDampingOverride=0.9999
    linearSpringConstant=10
    simulationSpace=World
    alphaBoolBlend {blendOption=Linear}
    boundBone {boneName=dyn_roll2_01_l}
    componentAppliedLinearAccClamp=<2000 2000 1000>
    componentLinearAccScale=<1 1 1>
    componentLinearVelScale=<1 1 1>
    gravityOverride {z=-980}
    physicsBodyDefinitions [{
      boundBone {boneName=dyn_roll2_01_l}
      boxExtents=<22 22 22>
      constraintSetup {
       angularTargetAxis=AxisY
       twistAxis=AxisY
       angularLimitsMax=<0.1 50 0.1>
       angularLimitsMin=<-0.1 -35 -0.1>
       angularTarget=<1 0 0>
       linearAxesMax=<0.1 0.1 0.1>
       linearAxesMin=<-0.1 -0.1 -0.1>
      }
      localJointOffset=<2.5 0 0>
     }]
   }
   n10 = Anim Dynamics - Bone: dyn_roll2_02_l(ComponentPose=@n9, Alpha=1, ExternalForce="0,0,0") {
    angularDampingOverride=0.9999
    angularSpringConstant=500
    bAngularSpring=true
    bLinearSpring=true
    bOverrideAngularDamping=true
    bOverrideLinearDamping=true
    bUseGravityOverride=true
    linearDampingOverride=0.9999
    linearSpringConstant=10
    simulationSpace=World
    alphaBoolBlend {blendOption=Linear}
    boundBone {boneName=dyn_roll2_02_l}
    componentAppliedLinearAccClamp=<2000 2000 2000>
    componentLinearAccScale=<1 1 1>
    componentLinearVelScale=<1 1 1>
    gravityOverride {z=-900}
    physicsBodyDefinitions [{
      boundBone {boneName=dyn_roll2_02_l}
      boxExtents=<17 17 17>
      constraintSetup {
       angularLimitsMax=<0.1 40 0.1>
       angularLimitsMin=<-0.1 -20 -0.1>
       angularTarget=<1 0 0>
       linearAxesMax=<0.1 0.1 0.1>
       linearAxesMin=<-0.1 -0.1 -0.1>
      }
      localJointOffset=<2 0 0>
     }]
   }
   n11 = Anim Dynamics - Bone: dyn_roll2_01_r(ComponentPose=@n10, Alpha=1, ExternalForce="0,0,0") {
    angularBiasOverride=0.9999
    angularDampingOverride=1
    angularSpringConstant=1000
    bAngularSpring=true
    bLinearSpring=true
    bOverrideAngularBias=true
    bOverrideAngularDamping=true
    bOverrideLinearDamping=true
    bUseGravityOverride=true
    linearDampingOverride=1
    linearSpringConstant=10
    simulationSpace=World
    alphaBoolBlend {blendOption=Linear}
    boundBone {boneName=dyn_roll2_01_r}
    componentAppliedLinearAccClamp=<1200 1200 1200>
    componentLinearAccScale=<0.7 0.7 0.8>
    componentLinearVelScale=<0.8 0.8 0.9>
    gravityOverride {z=-980}
    physicsBodyDefinitions [{
      boundBone {boneName=dyn_roll2_01_r}
      boxExtents=<17 17 17>
      constraintSetup {
       angularLimitsMax=<0.1 35 0.1>
       angularLimitsMin=<-0.1 -40 -0.1>
       angularTarget=<1 0 0>
       linearAxesMax=<0.1 0.1 0.1>
       linearAxesMin=<-0.1 -0.1 -0.1>
      }
      localJointOffset=<2 0 0>
     }]
   }
   n12 = Anim Dynamics - Bone: dyn_roll2_02_r(ComponentPose=@n11, Alpha=1, ExternalForce="0,0,0") {
    angularBiasOverride=0.9999
    angularDampingOverride=1
    angularSpringConstant=500
    bAngularSpring=true
    bLinearSpring=true
    bOverrideAngularBias=true
    bOverrideAngularDamping=true
    bOverrideLinearDamping=true
    bUseGravityOverride=true
    linearDampingOverride=1
    linearSpringConstant=10
    simulationSpace=World
    alphaBoolBlend {blendOption=Linear}
    boundBone {boneName=dyn_roll2_02_r}
    componentAppliedLinearAccClamp=<1000 1000 1000>
    componentLinearAccScale=<0.7 0.7 0.8>
    componentLinearVelScale=<0.8 0.8 0.8>
    gravityOverride {z=-900}
    physicsBodyDefinitions [{
      boundBone {boneName=dyn_roll2_02_r}
      boxExtents=<17 17 17>
      constraintSetup {
       angularLimitsMax=<0.1 30 0.1>
       angularLimitsMin=<-0.1 -40 -0.1>
       angularTarget=<1 0 0>
       linearAxesMax=<0.1 0.1 0.1>
       linearAxesMin=<-0.1 -0.1 -0.1>
      }
      localJointOffset=<1.5 0 0>
     }]
   }
   n13 = "Transform (Modify) Bone - Bone: dyn_roll2_MSTR_l"!(ComponentPose=@n12, bAlphaBoolEnabled:bool=True, Translation="0,0,0", Rotation="-35,0,0", Scale="1,1,1") -> (Pose) {
    alphaInputType=Bool
    rotationMode=BMM_Additive
    rotationSpace=BCS_ParentBoneSpace
    alphaBoolBlend {blendOption=Linear}
    boneToModify {boneName=dyn_roll2_MSTR_l}
    rotation {pitch=-35}
   }
   n14 = Anim Dynamics - Bone: dyn_roll3_01_l(ComponentPose=@n13, Alpha=0, ExternalForce="0,0,0") {
    angularDampingOverride=0.9999
    angularSpringConstant=1000
    bAngularSpring=true
    bLinearSpring=true
    bOverrideAngularDamping=true
    bOverrideLinearDamping=true
    bUseGravityOverride=true
    linearDampingOverride=0.9999
    linearSpringConstant=10
    simulationSpace=World
    alphaBoolBlend {blendOption=Linear}
    boundBone {boneName=dyn_roll3_01_l}
    componentAppliedLinearAccClamp=<2000 2000 1000>
    componentLinearAccScale=<1 1 1>
    componentLinearVelScale=<1 1 1>
    gravityOverride {z=-980}
    physicsBodyDefinitions [{
      boundBone {boneName=dyn_roll3_01_l}
      boxExtents=<22 22 22>
      constraintSetup {
       twistAxis=AxisY
       angularLimitsMax=<0.1 50 0.1>
       angularLimitsMin=<-0.1 -35 -0.1>
       angularTarget=<1 0 0>
       linearAxesMax=<0.1 0.1 0.1>
       linearAxesMin=<-0.1 -0.1 -0.1>
      }
      localJointOffset=<2.5 0 0>
     }]
   }
   n15 = Anim Dynamics - Bone: dyn_roll3_02_l(ComponentPose=@n14, Alpha=1, ExternalForce="0,0,0") {
    angularDampingOverride=0.9999
    angularSpringConstant=500
    bAngularSpring=true
    bLinearSpring=true
    bOverrideAngularDamping=true
    bOverrideLinearDamping=true
    bUseGravityOverride=true
    linearDampingOverride=0.9999
    linearSpringConstant=10
    simulationSpace=World
    alphaBoolBlend {blendOption=Linear}
    boundBone {boneName=dyn_roll3_02_l}
    componentAppliedLinearAccClamp=<1900 1900 1200>
    componentLinearAccScale=<1 1 1>
    componentLinearVelScale=<1 1 1>
    gravityOverride {z=-900}
    physicsBodyDefinitions [{
      boundBone {boneName=dyn_roll3_02_l}
      boxExtents=<22 22 22>
      constraintSetup {
       angularLimitsMax=<0.1 20 0.1>
       angularLimitsMin=<-0.1 -10 -0.1>
       angularTarget=<1 0 0>
       linearAxesMax=<0.1 0.1 0.1>
       linearAxesMin=<-0.1 -0.1 -0.1>
      }
      localJointOffset=<2.5 0 0>
     }]
   }
   n16 = Anim Dynamics - Bone: dyn_roll3_01_r(ComponentPose=@n15, Alpha=1, ExternalForce="0,0,0") {
    angularBiasOverride=0.9999
    angularDampingOverride=1
    angularSpringConstant=1000
    bAngularSpring=true
    bLinearSpring=true
    bOverrideAngularBias=true
    bOverrideAngularDamping=true
    bOverrideLinearDamping=true
    bUseGravityOverride=true
    linearDampingOverride=1
    linearSpringConstant=10
    simulationSpace=World
    alphaBoolBlend {blendOption=Linear}
    boundBone {boneName=dyn_roll3_01_r}
    componentAppliedLinearAccClamp=<1200 1200 1200>
    componentLinearAccScale=<0.7 0.7 0.8>
    componentLinearVelScale=<0.8 0.8 0.8>
    gravityOverride {z=-980}
    physicsBodyDefinitions [{
      boundBone {boneName=dyn_roll3_01_r}
      boxExtents=<17 12 12>
      constraintSetup {
       angularLimitsMax=<0.1 30 0.1>
       angularLimitsMin=<-0.1 -50 -0.1>
       angularTarget=<1 0 0>
       linearAxesMax=<0.1 0.1 0.1>
       linearAxesMin=<-0.1 -0.1 -0.1>
      }
      localJointOffset=<2 0 0>
     }]
   }
   n17 = Anim Dynamics - Bone: dyn_roll3_02_r(ComponentPose=@n16, Alpha=1, ExternalForce="0,0,0") {
    angularBiasOverride=0.9999
    angularDampingOverride=1
    angularSpringConstant=500
    bAngularSpring=true
    bLinearSpring=true
    bOverrideAngularBias=true
    bOverrideAngularDamping=true
    bOverrideLinearDamping=true
    bUseGravityOverride=true
    linearDampingOverride=1
    linearSpringConstant=10
    simulationSpace=World
    alphaBoolBlend {blendOption=Linear}
    boundBone {boneName=dyn_roll3_02_r}
    componentAppliedLinearAccClamp=<1000 1000 1000>
    componentLinearAccScale=<0.7 0.7 0.8>
    componentLinearVelScale=<0.8 0.8 0.8>
    gravityOverride {z=-900}
    physicsBodyDefinitions [{
      boundBone {boneName=dyn_roll3_02_r}
      boxExtents=<17 12 12>
      constraintSetup {
       angularLimitsMax=<0.1 30 0.1>
       angularLimitsMin=<-0.1 -40 -0.1>
       angularTarget=<1 0 0>
       linearAxesMax=<0.1 0.1 0.1>
       linearAxesMin=<-0.1 -0.1 -0.1>
      }
      localJointOffset=<1.5 0 0>
     }]
   }
   n18 = Anim Dynamics - Bone: dyn_roll4_01_l(ComponentPose=@n17, Alpha=1, ExternalForce="0,0,0") {
    angularBiasOverride=0.9999
    angularDampingOverride=1
    angularSpringConstant=1000
    bAngularSpring=true
    bLinearSpring=true
    bOverrideAngularBias=true
    bOverrideAngularDamping=true
    bOverrideLinearDamping=true
    bUseGravityOverride=true
    linearDampingOverride=1
    linearSpringConstant=10
    simulationSpace=World
    alphaBoolBlend {blendOption=Linear}
    boundBone {boneName=dyn_roll4_01_l}
    componentAppliedLinearAccClamp=<1500 1500 1500>
    componentLinearAccScale=<0.2 0.2 0.2>
    componentLinearVelScale=<0.3 0.3 0.3>
    gravityOverride {z=-980}
    physicsBodyDefinitions [{
      boundBone {boneName=dyn_roll4_01_l}
      boxExtents=<17 12 12>
      constraintSetup {
       angularLimitsMax=<0.1 40 0.1>
       angularLimitsMin=<-0.1 -25 -0.1>
       angularTarget=<1 0 0>
       linearAxesMax=<0.1 0.1 0.1>
       linearAxesMin=<-0.1 -0.1 -0.1>
      }
      localJointOffset=<1.5 0 0>
     }]
   }
   n19 = Anim Dynamics - Bone: dyn_roll4_02_l(ComponentPose=@n18, Alpha=1, ExternalForce="0,0,0") {
    angularBiasOverride=0.9999
    angularDampingOverride=1
    angularSpringConstant=500
    bAngularSpring=true
    bLinearSpring=true
    bOverrideAngularBias=true
    bOverrideAngularDamping=true
    bOverrideLinearDamping=true
    bUseGravityOverride=true
    linearDampingOverride=1
    linearSpringConstant=10
    simulationSpace=World
    alphaBoolBlend {blendOption=Linear}
    boundBone {boneName=dyn_roll4_02_l}
    componentAppliedLinearAccClamp=<1500 1500 1500>
    componentLinearAccScale=<0.2 0.2 0.2>
    componentLinearVelScale=<0.3 0.3 0.3>
    gravityOverride {z=-900}
    physicsBodyDefinitions [{
      boundBone {boneName=dyn_roll4_02_l}
      boxExtents=<12 12 12>
      constraintSetup {
       angularLimitsMax=<0.1 40 0.1>
       angularLimitsMin=<-0.1 -25 -0.1>
       angularTarget=<1 0 0>
       linearAxesMax=<0.1 0.1 0.1>
       linearAxesMin=<-0.1 -0.1 -0.1>
      }
      localJointOffset=<1.5 0 0>
     }]
   }
   n20 = Anim Dynamics - Bone: dyn_roll4_01_r(ComponentPose=@n19, Alpha=1, ExternalForce="0,0,0") {
    angularBiasOverride=0.9999
    angularDampingOverride=1
    angularSpringConstant=1000
    bAngularSpring=true
    bLinearSpring=true
    bOverrideAngularBias=true
    bOverrideAngularDamping=true
    bOverrideLinearDamping=true
    bUseGravityOverride=true
    linearDampingOverride=1
    linearSpringConstant=10
    simulationSpace=World
    alphaBoolBlend {blendOption=Linear}
    boundBone {boneName=dyn_roll4_01_r}
    componentAppliedLinearAccClamp=<1500 1500 1500>
    componentLinearAccScale=<0.2 0.2 0.2>
    componentLinearVelScale=<0.3 0.3 0.3>
    gravityOverride {z=-980}
    physicsBodyDefinitions [{
      boundBone {boneName=dyn_roll4_01_r}
      boxExtents=<12 17 17>
      constraintSetup {
       angularLimitsMax=<0.1 40 0.1>
       angularLimitsMin=<-0.1 -40 -0.1>
       angularTarget=<1 0 0>
       linearAxesMax=<0.1 0.1 0.1>
       linearAxesMin=<-0.1 -0.1 -0.1>
      }
      localJointOffset=<2 0 0>
     }]
   }
   n21 = Anim Dynamics - Bone: dyn_roll4_02_r(ComponentPose=@n20, Alpha=1, ExternalForce="0,0,0") {
    angularBiasOverride=0.9999
    angularDampingOverride=1
    angularSpringConstant=500
    bAngularSpring=true
    bLinearSpring=true
    bOverrideAngularBias=true
    bOverrideAngularDamping=true
    bOverrideLinearDamping=true
    bUseGravityOverride=true
    linearDampingOverride=1
    linearSpringConstant=10
    simulationSpace=World
    alphaBoolBlend {blendOption=Linear}
    boundBone {boneName=dyn_roll4_02_r}
    componentAppliedLinearAccClamp=<1500 1500 1500>
    componentLinearAccScale=<0.2 0.2 0.2>
    componentLinearVelScale=<0.3 0.3 0.3>
    gravityOverride {z=-900}
    physicsBodyDefinitions [{
      boundBone {boneName=dyn_roll4_02_r}
      boxExtents=<12 17 17>
      constraintSetup {
       angularLimitsMax=<0.1 30 0.1>
       angularLimitsMin=<-0.1 -40 -0.1>
       angularTarget=<1 0 0>
       linearAxesMax=<0.1 0.1 0.1>
       linearAxesMin=<-0.1 -0.1 -0.1>
      }
      localJointOffset=<1.5 0 0>
     }]
   }
   n22 = Anim Dynamics - Bone: dyn_handle_paper_01(ComponentPose=@n21, Alpha=1, ExternalForce="0,0,0") {
    angularDampingOverride=0.9999
    angularSpringConstant=2000
    bAngularSpring=true
    bLinearSpring=true
    bOverrideAngularDamping=true
    bOverrideLinearDamping=true
    bUseGravityOverride=true
    linearDampingOverride=0.9999
    linearSpringConstant=10
    simulationSpace=World
    alphaBoolBlend {blendOption=Linear}
    boundBone {boneName=dyn_handle_paper_01}
    componentAppliedLinearAccClamp=<2000 2000 2000>
    componentLinearAccScale=<0.1 0.1 0.1>
    componentLinearVelScale=<0.2 0.2 0.2>
    gravityOverride {z=-980}
    physicsBodyDefinitions [{
      boundBone {boneName=dyn_handle_paper_01}
      boxExtents=<20 20 20>
      constraintSetup {angularLimitsMax=<10 40 10>, angularLimitsMin=<-10 -40 -30>, angularTarget=<1 0 0>}
      localJointOffset=<1 0 0>
     }]
   }
   n23 = Anim Dynamics - Bone: dyn_roll_01_l(ComponentPose=Local To Component(LocalPose=Input Pose() {name=InPose_0}), Alpha=1, ExternalForce="0,0,0") {
    angularDampingOverride=0.9999
    angularSpringConstant=1000
    bAngularSpring=true
    bOverrideAngularDamping=true
    bUseGravityOverride=true
    alphaBoolBlend {blendOption=Linear}
    boundBone {boneName=dyn_roll_01_l}
    componentAppliedLinearAccClamp=<2000 2000 1000>
    componentLinearAccScale=<1 1 1>
    componentLinearVelScale=<1 1 1>
    gravityOverride {z=-900}
    physicsBodyDefinitions [{
      boundBone {boneName=dyn_roll_01_l}
      boxExtents=<20 20 20>
      constraintSetup {angularLimitsMax=<5 50 5>, angularLimitsMin=<-5 -35 -5>, angularTarget=<1 0 0>}
      localJointOffset=<2 0 0>
     }]
   }
   n24 = Anim Dynamics - Bone: dyn_roll_02_l(ComponentPose=@n23, Alpha=1, ExternalForce="0,0,0") {
    angularDampingOverride=0.9999
    angularSpringConstant=500
    bAngularSpring=true
    bOverrideAngularDamping=true
    bUseGravityOverride=true
    alphaBoolBlend {blendOption=Linear}
    boundBone {boneName=dyn_roll_02_l}
    componentAppliedLinearAccClamp=<2000 2000 2000>
    componentLinearAccScale=<1 1 1>
    componentLinearVelScale=<1 1 1>
    gravityOverride {z=-800}
    physicsBodyDefinitions [{
      boundBone {boneName=dyn_roll_02_l}
      boxExtents=<15 20 20>
      constraintSetup {angularLimitsMax=<5 40 5>, angularLimitsMin=<-5 -20 -5>, angularTarget=<1 0 0>}
      localJointOffset=<2 0 0>
     }]
   }
   n25 = Anim Dynamics - Bone: dyn_roll_01_r(ComponentPose=@n24, Alpha=1, ExternalForce="0,0,0") {
    angularDampingOverride=1
    angularSpringConstant=1000
    bAngularSpring=true
    bOverrideAngularDamping=true
    bUseGravityOverride=true
    alphaBoolBlend {blendOption=Linear}
    boundBone {boneName=dyn_roll_01_r}
    componentAppliedLinearAccClamp=<1200 1200 1250>
    componentLinearAccScale=<0.7 0.7 0.8>
    componentLinearVelScale=<0.8 0.8 0.9>
    gravityOverride {z=-950}
    physicsBodyDefinitions [{
      boundBone {boneName=dyn_roll_01_r}
      boxExtents=<15 15 15>
      constraintSetup {angularLimitsMax=<5 30 5>, angularLimitsMin=<-5 -50 -5>, angularTarget=<1 0 0>}
      localJointOffset=<2.5 0 0>
     }]
   }
   n26 = Anim Dynamics - Bone: dyn_roll_02_r(ComponentPose=@n25, Alpha=1, ExternalForce="0,0,0") {
    angularDampingOverride=1
    angularSpringConstant=500
    bAngularSpring=true
    bOverrideAngularDamping=true
    bUseGravityOverride=true
    alphaBoolBlend {blendOption=Linear}
    boundBone {boneName=dyn_roll_02_r}
    componentAppliedLinearAccClamp=<1000 1000 1000>
    componentLinearAccScale=<0.7 0.7 0.8>
    componentLinearVelScale=<0.8 0.8 0.8>
    gravityOverride {z=-900}
    physicsBodyDefinitions [{
      boundBone {boneName=dyn_roll_02_r}
      boxExtents=<10 15 15>
      constraintSetup {angularLimitsMax=<5 10 5>, angularLimitsMin=<-5 -40 -5>, angularTarget=<1 0 0>}
      localJointOffset=<2.5 0 0>
     }]
   }
   n27 = Anim Dynamics - Bone: dyn_roll2_01_l(ComponentPose=@n26, Alpha=1, ExternalForce="0,0,0") {
    angularDampingOverride=0.9999
    angularSpringConstant=1000
    bAngularSpring=true
    bOverrideAngularDamping=true
    bUseGravityOverride=true
    alphaBoolBlend {blendOption=Linear}
    boundBone {boneName=dyn_roll2_01_l}
    componentAppliedLinearAccClamp=<2000 2000 1000>
    componentLinearAccScale=<1 1 1>
    componentLinearVelScale=<1 1 1>
    gravityOverride {z=-900}
    physicsBodyDefinitions [{
      boundBone {boneName=dyn_roll2_01_l}
      boxExtents=<20 20 20>
      constraintSetup {angularLimitsMax=<5 50 0>, angularLimitsMin=<-5 -35 0>, angularTarget=<1 0 0>}
      localJointOffset=<2 0 0>
     }]
   }
   n28 = Anim Dynamics - Bone: dyn_roll2_02_l(ComponentPose=@n27, Alpha=1, ExternalForce="0,0,0") {
    angularDampingOverride=0.9999
    angularSpringConstant=500
    bAngularSpring=true
    bOverrideAngularDamping=true
    bUseGravityOverride=true
    alphaBoolBlend {blendOption=Linear}
    boundBone {boneName=dyn_roll2_02_l}
    componentAppliedLinearAccClamp=<2000 2000 2000>
    componentLinearAccScale=<1 1 1>
    componentLinearVelScale=<1 1 1>
    gravityOverride {z=-800}
    physicsBodyDefinitions [{
      boundBone {boneName=dyn_roll2_02_l}
      boxExtents=<15 20 20>
      constraintSetup {angularLimitsMax=<5 40 5>, angularLimitsMin=<-5 -20 -5>, angularTarget=<1 0 0>}
      localJointOffset=<2 0 0>
     }]
   }
   n29 = Anim Dynamics - Bone: dyn_roll2_03_l(ComponentPose=@n28, Alpha=1, ExternalForce="0,0,0") {
    angularDampingOverride=0.9999
    angularSpringConstant=250
    bAngularSpring=true
    bOverrideAngularDamping=true
    bUseGravityOverride=true
    alphaBoolBlend {blendOption=Linear}
    boundBone {boneName=dyn_roll2_03_l}
    componentAppliedLinearAccClamp=<2000 2000 2000>
    componentLinearAccScale=<1 1 1>
    componentLinearVelScale=<1 1 1>
    gravityOverride {z=-750}
    physicsBodyDefinitions [{
      boundBone {boneName=dyn_roll2_03_l}
      boxExtents=<15 15 20>
      constraintSetup {angularLimitsMax=<5 40 5>, angularLimitsMin=<-5 -20 -5>, angularTarget=<1 0 0>}
      localJointOffset=<2 0 0>
     }]
   }
   n30 = Anim Dynamics - Bone: dyn_roll2_01_r(ComponentPose=@n29, Alpha=1, ExternalForce="0,0,0") {
    angularDampingOverride=1
    angularSpringConstant=1000
    bAngularSpring=true
    bOverrideAngularDamping=true
    bUseGravityOverride=true
    alphaBoolBlend {blendOption=Linear}
    boundBone {boneName=dyn_roll2_01_r}
    componentAppliedLinearAccClamp=<1200 1200 1200>
    componentLinearAccScale=<0.7 0.7 0.8>
    componentLinearVelScale=<0.8 0.8 0.9>
    gravityOverride {z=-950}
    physicsBodyDefinitions [{
      boundBone {boneName=dyn_roll2_01_r}
      boxExtents=<15 15 15>
      constraintSetup {angularLimitsMax=<5 30 0>, angularLimitsMin=<-5 -50 0>, angularTarget=<1 0 0>}
      localJointOffset=<2.5 0 0>
     }]
   }
   n31 = Anim Dynamics - Bone: dyn_roll2_02_r(ComponentPose=@n30, Alpha=1, ExternalForce="0,0,0") {
    angularDampingOverride=1
    angularSpringConstant=500
    bAngularSpring=true
    bOverrideAngularDamping=true
    bUseGravityOverride=true
    alphaBoolBlend {blendOption=Linear}
    boundBone {boneName=dyn_roll2_02_r}
    componentAppliedLinearAccClamp=<1000 1000 1000>
    componentLinearAccScale=<0.7 0.7 0.8>
    componentLinearVelScale=<0.8 0.8 0.8>
    gravityOverride {z=-900}
    physicsBodyDefinitions [{
      boundBone {boneName=dyn_roll2_02_r}
      boxExtents=<10 15 15>
      constraintSetup {angularLimitsMax=<5 10 5>, angularLimitsMin=<-5 -40 -5>, angularTarget=<1 0 0>}
      localJointOffset=<2.5 0 0>
     }]
   }
   n32 = Anim Dynamics - Bone: dyn_roll3_01_l(ComponentPose=@n31, Alpha=1, ExternalForce="0,0,0") {
    angularDampingOverride=0.9999
    angularSpringConstant=1000
    bAngularSpring=true
    bOverrideAngularDamping=true
    bUseGravityOverride=true
    alphaBoolBlend {blendOption=Linear}
    boundBone {boneName=dyn_roll3_01_l}
    componentAppliedLinearAccClamp=<2000 2000 1000>
    componentLinearAccScale=<1 1 1>
    componentLinearVelScale=<1 1 1>
    gravityOverride {z=-900}
    physicsBodyDefinitions [{
      boundBone {boneName=dyn_roll3_01_l}
      boxExtents=<20 20 20>
      constraintSetup {angularLimitsMax=<2 50 0>, angularLimitsMin=<-2 -40 0>, angularTarget=<1 0 0>}
      localJointOffset=<2 0 0>
     }]
   }
   n33 = Anim Dynamics - Bone: dyn_roll3_02_l(ComponentPose=@n32, Alpha=1, ExternalForce="0,0,0") {
    angularDampingOverride=0.9999
    angularSpringConstant=500
    bAngularSpring=true
    bOverrideAngularDamping=true
    bUseGravityOverride=true
    alphaBoolBlend {blendOption=Linear}
    boundBone {boneName=dyn_roll3_02_l}
    componentAppliedLinearAccClamp=<1900 1900 1200>
    componentLinearAccScale=<1 1 1>
    componentLinearVelScale=<1 1 1>
    gravityOverride {z=-800}
    physicsBodyDefinitions [{
      boundBone {boneName=dyn_roll3_02_l}
      boxExtents=<15 20 20>
      constraintSetup {angularLimitsMax=<2 40 5>, angularLimitsMin=<-2 -30 -5>, angularTarget=<1 0 0>}
      localJointOffset=<2 0 0>
     }]
   }
   n34 = Anim Dynamics - Bone: dyn_roll3_03_l(ComponentPose=@n33, Alpha=1, ExternalForce="0,0,0") {
    angularDampingOverride=0.9999
    angularSpringConstant=250
    bAngularSpring=true
    bOverrideAngularDamping=true
    bUseGravityOverride=true
    alphaBoolBlend {blendOption=Linear}
    boundBone {boneName=dyn_roll3_03_l}
    componentAppliedLinearAccClamp=<1800 1800 1800>
    componentLinearAccScale=<1 1 1>
    componentLinearVelScale=<1 1 1>
    gravityOverride {z=-750}
    physicsBodyDefinitions [{
      boundBone {boneName=dyn_roll3_03_l}
      boxExtents=<15 15 20>
      constraintSetup {angularLimitsMax=<2 20 5>, angularLimitsMin=<-2 -20 -5>, angularTarget=<1 0 0>}
      localJointOffset=<2 0 0>
     }]
   }
   n35 = Anim Dynamics - Bone: dyn_roll3_01_r(ComponentPose=@n34, Alpha=1, ExternalForce="0,0,0") {
    angularDampingOverride=1
    angularSpringConstant=1000
    bAngularSpring=true
    bOverrideAngularDamping=true
    bUseGravityOverride=true
    alphaBoolBlend {blendOption=Linear}
    boundBone {boneName=dyn_roll3_01_r}
    componentAppliedLinearAccClamp=<1200 1200 1200>
    componentLinearAccScale=<0.7 0.7 0.8>
    componentLinearVelScale=<0.8 0.8 0.8>
    gravityOverride {z=-950}
    physicsBodyDefinitions [{
      boundBone {boneName=dyn_roll3_01_r}
      boxExtents=<15 15 15>
      constraintSetup {angularLimitsMax=<5 30 5>, angularLimitsMin=<-5 -50 -5>, angularTarget=<1 0 0>}
      localJointOffset=<2.5 0 0>
     }]
   }
   n36 = Anim Dynamics - Bone: dyn_roll3_02_r(ComponentPose=@n35, Alpha=1, ExternalForce="0,0,0") {
    angularDampingOverride=1
    angularSpringConstant=500
    bAngularSpring=true
    bOverrideAngularDamping=true
    bUseGravityOverride=true
    alphaBoolBlend {blendOption=Linear}
    boundBone {boneName=dyn_roll3_02_r}
    componentAppliedLinearAccClamp=<1000 1000 1000>
    componentLinearAccScale=<0.7 0.7 0.8>
    componentLinearVelScale=<0.8 0.8 0.8>
    gravityOverride {z=-900}
    physicsBodyDefinitions [{
      boundBone {boneName=dyn_roll3_02_r}
      boxExtents=<10 15 15>
      constraintSetup {angularLimitsMax=<5 10 5>, angularLimitsMin=<-5 -40 -5>, angularTarget=<1 0 0>}
      localJointOffset=<2.5 0 0>
     }]
   }
   n37 = Anim Dynamics - Bone: dyn_roll4_01_l(ComponentPose=@n36, Alpha=1, ExternalForce="0,0,0") {
    angularDampingOverride=0.9999
    angularSpringConstant=1000
    bAngularSpring=true
    bOverrideAngularDamping=true
    bUseGravityOverride=true
    alphaBoolBlend {blendOption=Linear}
    boundBone {boneName=dyn_roll4_01_l}
    componentAppliedLinearAccClamp=<1800 1800 1800>
    componentLinearAccScale=<1 1 1>
    componentLinearVelScale=<1 1 1>
    gravityOverride {z=-900}
    physicsBodyDefinitions [{
      boundBone {boneName=dyn_roll4_01_l}
      boxExtents=<15 20 20>
      constraintSetup {angularLimitsMax=<5 60 -2.5>, angularLimitsMin=<-5 -25 2.5>, angularTarget=<1 0 0>}
      localJointOffset=<2 0 0>
     }]
   }
   n38 = Anim Dynamics - Bone: dyn_roll4_02_l(ComponentPose=@n37, Alpha=1, ExternalForce="0,0,0") {
    angularDampingOverride=0.9999
    angularSpringConstant=500
    bAngularSpring=true
    bOverrideAngularDamping=true
    bUseGravityOverride=true
    alphaBoolBlend {blendOption=Linear}
    boundBone {boneName=dyn_roll4_02_l}
    componentAppliedLinearAccClamp=<1700 1700 1700>
    componentLinearAccScale=<1 1 1>
    componentLinearVelScale=<1 1 1>
    gravityOverride {z=-800}
    physicsBodyDefinitions [{
      boundBone {boneName=dyn_roll4_02_l}
      boxExtents=<15 15 20>
      constraintSetup {angularLimitsMax=<5 40 5>, angularLimitsMin=<-5 -25 -5>, angularTarget=<1 0 0>}
      localJointOffset=<2 0 0>
     }]
   }
   n39 = Anim Dynamics - Bone: dyn_roll4_01_r(ComponentPose=@n38, Alpha=1, ExternalForce="0,0,0") {
    angularDampingOverride=1
    angularSpringConstant=1000
    bAngularSpring=true
    bOverrideAngularDamping=true
    bUseGravityOverride=true
    alphaBoolBlend {blendOption=Linear}
    boundBone {boneName=dyn_roll4_01_r}
    componentAppliedLinearAccClamp=<1200 1200 1200>
    componentLinearAccScale=<0.7 0.7 0.8>
    componentLinearVelScale=<0.8 0.8 0.8>
    gravityOverride {z=-950}
    physicsBodyDefinitions [{
      boundBone {boneName=dyn_roll4_01_r}
      boxExtents=<15 15 15>
      constraintSetup {angularLimitsMax=<5 20 -2.5>, angularLimitsMin=<-5 -60 2.5>, angularTarget=<1 0 0>}
      localJointOffset=<2.5 0 0>
     }]
   }
   n40 = Anim Dynamics - Bone: dyn_roll4_02_r(ComponentPose=@n39, Alpha=1, ExternalForce="0,0,0") {
    angularDampingOverride=1
    angularSpringConstant=500
    bAngularSpring=true
    bOverrideAngularDamping=true
    bUseGravityOverride=true
    alphaBoolBlend {blendOption=Linear}
    boundBone {boneName=dyn_roll4_02_r}
    componentAppliedLinearAccClamp=<1000 1000 1000>
    componentLinearAccScale=<0.7 0.7 0.8>
    componentLinearVelScale=<0.8 0.8 0.8>
    gravityOverride {z=-900}
    physicsBodyDefinitions [{
      boundBone {boneName=dyn_roll4_02_r}
      boxExtents=<10 15 15>
      constraintSetup {angularLimitsMax=<5 20 5>, angularLimitsMin=<-5 -60 -5>, angularTarget=<1 0 0>}
      localJointOffset=<2.5 0 0>
     }]
   }
   n41 = Anim Dynamics - Bone: dyn_handle_paper_01(ComponentPose=@n40, Alpha=1, ExternalForce="0,0,0") {
    angularDampingOverride=0.9999
    angularSpringConstant=1000
    bAngularSpring=true
    bOverrideAngularDamping=true
    bUseGravityOverride=true
    alphaBoolBlend {blendOption=Linear}
    boundBone {boneName=dyn_handle_paper_01}
    componentAppliedLinearAccClamp=<2000 2000 2000>
    componentLinearAccScale=<1 1 1>
    componentLinearVelScale=<1 1 1>
    gravityOverride {z=-900}
    physicsBodyDefinitions [{
      boundBone {boneName=dyn_handle_paper_01}
      boxExtents=<20 20 20>
      constraintSetup {angularLimitsMax=<10 70 20>, angularLimitsMin=<-10 -40 -20>, angularTarget=<1 0 0>}
      localJointOffset=<2 0 0>
     }]
   }
   Output Pose(Result=Blend Poses by bool(BlendPose_0=Component To Local(ComponentPose=@n22), BlendPose_1=Component To Local(ComponentPose=@n41), bActiveValue=Get bIsInFrontEnd()) {blendType=Linear}) {layerGroup=None}
  ]}
 EventGraph {nodes [
   Event SetStage01() {is=title K2Node_CustomEvent}
   > n1 = Get Dynamic Material Instance(SkeletalMeshComponent=GetPartSkeletalMeshComponent())
   > SetScalarParameterValue(self=@n1, ParameterName=Hide Element 01, Value=0)
   > SetScalarParameterValue(self=@n1, ParameterName=Hide Element 02, Value=1)
   > SetScalarParameterValue(self=@n1, ParameterName=Hide Element 03, Value=1)
   > SetScalarParameterValue(self=@n1, ParameterName=Hide Element 04, Value=1)
   > ResetRollsStages()
   > Set IsRoll1Active(IsRoll1Active:bool=true) -> (Output_Get:bool)
   Event SetStage02() {is=title K2Node_CustomEvent}
   > n2 = Get Dynamic Material Instance(SkeletalMeshComponent=GetPartSkeletalMeshComponent())
   > SetScalarParameterValue(self=@n2, ParameterName=Hide Element 01, Value=1)
   > SetScalarParameterValue(self=@n2, ParameterName=Hide Element 02, Value=0)
   > SetScalarParameterValue(self=@n2, ParameterName=Hide Element 03, Value=1)
   > SetScalarParameterValue(self=@n2, ParameterName=Hide Element 04, Value=1)
   > ResetRollsStages()
   > Set IsRoll2Active(IsRoll2Active:bool=true) -> (Output_Get:bool)
   Event SetStage03() {is=title K2Node_CustomEvent}
   > n3 = Get Dynamic Material Instance(SkeletalMeshComponent=GetPartSkeletalMeshComponent())
   > SetScalarParameterValue(self=@n3, ParameterName=Hide Element 01, Value=1)
   > SetScalarParameterValue(self=@n3, ParameterName=Hide Element 02, Value=1)
   > SetScalarParameterValue(self=@n3, ParameterName=Hide Element 03, Value=0)
   > SetScalarParameterValue(self=@n3, ParameterName=Hide Element 04, Value=1)
   > ResetRollsStages()
   > Set IsRoll3Active(IsRoll3Active:bool=true) -> (Output_Get:bool)
   Event SetStage04() {is=title K2Node_CustomEvent}
   > n4 = Get Dynamic Material Instance(SkeletalMeshComponent=GetPartSkeletalMeshComponent())
   > SetScalarParameterValue(self=@n4, ParameterName=Hide Element 01, Value=1)
   > SetScalarParameterValue(self=@n4, ParameterName=Hide Element 02, Value=1)
   > SetScalarParameterValue(self=@n4, ParameterName=Hide Element 03, Value=1)
   > SetScalarParameterValue(self=@n4, ParameterName=Hide Element 04, Value=0)
   > ResetRollsStages()
   > Set IsRoll4Active(IsRoll4Active:bool=true) -> (Output_Get:bool)
   Event BlueprintUpdateAnimation()
   > n5 = Get!(execute:Exec) -> (then:Exec, else:Exec, BodyMesh:Skeletal Mesh Component Budgeted Object Reference)
   > Set CharacterHoldingPickaxe(CharacterHoldingPickaxe:bool=true) -> (Output_Get:bool)
   Set CharacterHoldingPickaxe(execute=@n5.else, CharacterHoldingPickaxe=false)
   n6 = Event OnCharacterCustomizationLoaded()
   > Set BodyMesh(BodyMesh:Skeletal Mesh Component Budgeted Object Reference=Get Array Item with EFort Custom Part Type(Array=@n6.CustomMeshParts)) -> (Output_Get:Skeletal Mesh Component Budgeted Object Reference)
   Branch(Condition=Get bIsPlayingMeleeAnim())
   > Set TimeAccumulator(TimeAccumulator:double=float + float(A=Get TimeAccumulator() -> (TimeAccumulator:double), B=Get Delta Time X())) -> (Output_Get:double)
   > Branch(Condition="float >= float"(A=Get TimeAccumulator(), B=2))
   > Set StageCount(StageCount:int=int + int(A=Get StageCount() -> (StageCount:int=0))) -> (Output_Get:int=0)
   > Set TimeAccumulator(TimeAccumulator=0)
   > n7 = Branch(Condition="integer > integer"(A=Get StageCount() -> (StageCount=0), B=4))
   > Set StageCount(StageCount=4) -> (Output_Get=0)
   SetStageByIndex(execute=@n7.else)
   n8 = Switch on Int!(execute:Exec, Selection:int=Get StageCount() -> (StageCount=0)) -> (Default:Exec, 0:Exec, 1:Exec, 2:Exec, 3:Exec, 4:Exec)
   Event SetStage01(execute=@n8.1) {is=title K2Node_CallFunction}
   Event SetStage02(execute=@n8.2) {is=title K2Node_CallFunction}
   Event SetStage03(execute=@n8.3) {is=title K2Node_CallFunction}
   Event SetStage04(execute=@n8.4) {is=title K2Node_CallFunction}
   Set Delta Time X(Delta Time X=0)
   > AddHit()
   Set StageCount(StageCount=1) -> (Output_Get=0)
   > Set TimeAccumulator(TimeAccumulator=0)
   SetStageByIndex()
   Event SetStage01()
   SetStageByIndex() {is=title K2Node_CustomEvent}
   AddHit() {is=title K2Node_CustomEvent}
   Event OnWeaponChanged()
   Set StageCount(StageCount=1) -> (Output_Get=0)
   Set TimeAccumulator(TimeAccumulator=0)
  ]}
 ResetRollsStages {nodes [
   ResetRollsStages() {is=title K2Node_FunctionEntry}
   > Set IsRoll1Active(IsRoll1Active=false)
   > Set IsRoll2Active(IsRoll2Active=false)
   > Set IsRoll3Active(IsRoll3Active=false)
   > Set IsRoll4Active(IsRoll4Active=false)
  ]}
]
variables [
 BodyMesh {type=SkeletalMeshComponentBudgeted}
 CharacterHoldingPickaxe {type=boolean}
 Delta Time X {type=number}
 IsRoll1Active {type=boolean}
 IsRoll2Active {type=boolean}
 IsRoll3Active {type=boolean}
 IsRoll4Active {type=boolean}
 StageCount {type=integer, default=1}
 TimeAccumulator {type=number}
]
