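# Physics asset for the TwilightSpotSpell two-handed pickaxe, defining bodies and joints for its dangling roll and paper pieces.
# Eight roll chains, four per side, start at large MSTR capsules with free angular joints to the handle.
# Their two or three links each use three parallel capsules and hinge-like limits, bending mainly on swing2.
# A small handle-paper chain has an angularly locked root and one limited swinging link.
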PhysicsAsset /BRCosmetics/Weapons/FORT_Melee/Pickaxe_TwilightSpotSpell/Meshes/TwilightSpotSpell_Axe_PhysicsAsset
bodies {
 Melee_TwoHanded_Handle [{
   shapeType=Capsule
   length=34.0784
   radius=3.5075
   shapeName=Melee_TwoHanded_Handle_capsule
   center=<-0.11 -1.0452 21.9066>
   rotation=<rot 0.6212 -90 -1.2241>
  }]
 dyn_handle_paper_01 [{
   shapeType=Capsule
   length=2.6389
   radius=4.9652
   shapeName=dyn_handle_paper_01_capsule
   center=<9.119 0.3583 -0.4132>
   rotation=<rot -0.894 -86.2113 -89.5221>
  }]
 dyn_handle_paper_MSTR [{shapeType=Capsule, length=0.5, radius=0.5, shapeName=dyn_handle_paper_MSTR_capsule, rotation=<rot 0 -90 0>}]
 dyn_roll2_01_l [
  {shapeType=Capsule, length=5, radius=2, center=<2 -10 -1.5899>, rotation=<rot 90 -90 -89.9999>}
  {shapeType=Capsule, length=5, radius=2, center=<2 0 -1.5899>, rotation=<rot 90 -179.9999 -179.9998>}
  {shapeType=Capsule, length=5, radius=2, center=<2 10 -1.5899>, rotation=<rot 90 -90 -89.9999>}
 ]
 dyn_roll2_01_r [
  {shapeType=Capsule, length=5, radius=2, center=<5 -10 1.214>, rotation=<rot 90 -160.5288 -160.5286>}
  {shapeType=Capsule, length=5, radius=2, center=<5 0 1.214>, rotation=<rot 90 -179.9999 -179.9998>}
  {shapeType=Capsule, length=5, radius=2, center=<5 10 1.214>, rotation=<rot 90 -160.5288 -160.5286>}
 ]
 dyn_roll2_02_l [
  {shapeType=Capsule, length=5, radius=2, shapeName=dyn_roll_02_l_capsule, center=<2 -10 -0.8256>, rotation=<rot 87.1875 180 -180>}
  {shapeType=Capsule, length=5, radius=2, shapeName=dyn_roll_02_l_capsule, center=<2 0 -0.8256>, rotation=<rot 87.1875 180 -180>}
  {shapeType=Capsule, length=5, radius=2, shapeName=dyn_roll_02_l_capsule, center=<2 10 -0.8256>, rotation=<rot 87.1875 180 -180>}
 ]
 dyn_roll2_02_r [
  {shapeType=Capsule, length=5, radius=2, shapeName=dyn_roll_02_l_capsule, center=<5 -10 0.8312>, rotation=<rot 90 172.4476 172.4478>}
  {shapeType=Capsule, length=5, radius=2, shapeName=dyn_roll_02_l_capsule, center=<5 0 0.8312>, rotation=<rot 90 172.4476 172.4478>}
  {shapeType=Capsule, length=5, radius=2, shapeName=dyn_roll_02_l_capsule, center=<5 10 0.8312>, rotation=<rot 90 172.4476 172.4478>}
 ]
 dyn_roll2_03_l [
  {shapeType=Capsule, length=5, radius=2, center=<5 -10 -1.5899>, rotation=<rot 90 -179.9994 -179.9993>}
  {shapeType=Capsule, length=5, radius=2, shapeName=dyn_roll_02_l_capsule, center=<5 0 -0.6566>, rotation=<rot 87.1875 180 -180>}
  {shapeType=Capsule, length=5, radius=2, shapeName=dyn_roll_02_l_capsule, center=<5 10 -0.6566>, rotation=<rot 87.1875 180 -180>}
 ]
 dyn_roll2_MSTR_l [{shapeType=Capsule, length=14, radius=12, shapeName=dyn_roll_MSTR_l_capsule, center=<-2.0972 1.2408 12.9605>, rotation=<rot 90 180 -90>}]
 dyn_roll2_MSTR_r [{shapeType=Capsule, length=14.5, radius=12.5, shapeName=dyn_roll_MSTR_l_capsule, center=<-2.7363 0.5267 -12.147>, rotation=<rot 90 180 -90>}]
 dyn_roll3_01_l [
  {shapeType=Capsule, length=5, radius=2, center=<5 -10 -1.5513>, rotation=<rot 87.1875 179.9999 -180>}
  {shapeType=Capsule, length=5, radius=2, center=<5 0 -1.5513>, rotation=<rot 87.1875 179.9999 -180>}
  {shapeType=Capsule, length=5, radius=2, center=<5 10 -1.5513>, rotation=<rot 87.1875 179.9999 -180>}
 ]
 dyn_roll3_01_r [
  {shapeType=Capsule, length=5, radius=2, center=<5 -10 1.3536>, rotation=<rot 90 -160.5288 -160.5286>}
  {shapeType=Capsule, length=5, radius=2, center=<5 0 1.3536>, rotation=<rot 90 -179.9999 -179.9998>}
  {shapeType=Capsule, length=5, radius=2, center=<5 10 1.3536>, rotation=<rot 90 -160.5288 -160.5286>}
 ]
 dyn_roll3_02_l [
  {shapeType=Capsule, length=5, radius=2, shapeName=dyn_roll_02_l_capsule, center=<5 -10 -0.9224>, rotation=<rot 87.1875 180 -180>}
  {shapeType=Capsule, length=5, radius=2, shapeName=dyn_roll_02_l_capsule, center=<5 0 -0.9224>, rotation=<rot 87.1875 180 -180>}
  {shapeType=Capsule, length=5, radius=2, shapeName=dyn_roll_02_l_capsule, center=<5 10 -0.9224>, rotation=<rot 87.1875 180 -180>}
 ]
 dyn_roll3_02_r [
  {shapeType=Capsule, length=5, radius=2, shapeName=dyn_roll_02_l_capsule, center=<5 -10 1.1939>, rotation=<rot 90 160.5288 160.529>}
  {shapeType=Capsule, length=5, radius=2, shapeName=dyn_roll_02_l_capsule, center=<5 0 1.1939>, rotation=<rot 90 160.5288 160.529>}
  {shapeType=Capsule, length=5, radius=2, shapeName=dyn_roll_02_l_capsule, center=<5 10 1.1939>, rotation=<rot 90 160.5288 160.529>}
 ]
 dyn_roll3_03_l [
  {shapeType=Capsule, length=5, radius=2, shapeName=dyn_roll_02_l_capsule, center=<5 -10 -0.9224>, rotation=<rot 87.1875 180 -180>}
  {shapeType=Capsule, length=5, radius=2, shapeName=dyn_roll_02_l_capsule, center=<5 0 -0.9224>, rotation=<rot 87.1875 180 -180>}
  {shapeType=Capsule, length=5, radius=2, shapeName=dyn_roll_02_l_capsule, center=<5 10 -0.9224>, rotation=<rot 87.1875 180 -180>}
 ]
 dyn_roll3_MSTR_l [{
   shapeType=Capsule
   length=17
   radius=9
   shapeName=dyn_roll_MSTR_l_capsule
   center=<-1.5892 1.2408 10.2111>
   rotation=<rot 90 153.4349 -116.5651>
  }]
 dyn_roll3_MSTR_r [{
   shapeType=Capsule
   length=17
   radius=9
   shapeName=dyn_roll_MSTR_l_capsule
   center=<-0.7817 0.5267 -10.8494>
   rotation=<rot 90 153.4349 -116.5651>
  }]
 dyn_roll4_01_l [
  {shapeType=Capsule, length=5, radius=2, center=<5 -10 -0.9844>, rotation=<rot 87.1875 179.9996 179.9999>}
  {shapeType=Capsule, length=5, radius=2, center=<5 0 -0.9844>, rotation=<rot 87.1875 179.9996 179.9999>}
  {shapeType=Capsule, length=5, radius=2, center=<5 10 -0.9844>, rotation=<rot 87.1875 179.9996 179.9999>}
 ]
 dyn_roll4_01_r [
  {shapeType=Capsule, length=5, radius=2, center=<7.049 -10 0.9267>, rotation=<rot 87.1875 0.0007 0.0009>}
  {shapeType=Capsule, length=5, radius=2, center=<7.049 0 0.9267>, rotation=<rot 87.1875 0.0006 0.0008>}
  {shapeType=Capsule, length=5, radius=2, center=<7.049 10 0.9267>, rotation=<rot 87.1875 0.0006 0.0007>}
 ]
 dyn_roll4_02_l [
  {shapeType=Capsule, length=5, radius=2, center=<5 -10 -0.9844>, rotation=<rot 87.1875 179.9996 179.9999>}
  {shapeType=Capsule, length=5, radius=2, center=<5 0 -0.9844>, rotation=<rot 87.1875 179.9996 179.9999>}
  {shapeType=Capsule, length=5, radius=2, center=<5 10 -0.9844>, rotation=<rot 87.1875 179.9996 179.9999>}
 ]
 dyn_roll4_02_r [
  {shapeType=Capsule, length=5, radius=2, center=<7.049 -10 0.9267>, rotation=<rot 87.1875 0.0007 0.0009>}
  {shapeType=Capsule, length=5, radius=2, center=<7.049 0 0.9267>, rotation=<rot 87.1875 0.0006 0.0008>}
  {shapeType=Capsule, length=5, radius=2, center=<7.049 10 0.9267>, rotation=<rot 87.1875 0.0006 0.0008>}
 ]
 dyn_roll4_MSTR_l [{shapeType=Capsule, length=17, radius=6, shapeName=dyn_roll_MSTR_l_capsule, center=<-0.976 1.2408 9.5472>, rotation=<rot 90 180 -90>}]
 dyn_roll4_MSTR_r [{
   shapeType=Capsule
   length=20
   radius=6
   shapeName=dyn_roll_MSTR_l_capsule
   center=<-2.5526 0.5267 -6.6922>
   rotation=<rot 90 153.4349 -116.5651>
  }]
 dyn_roll_01_l [
  {shapeType=Capsule, length=7.5, radius=2, center=<5 -10 -1.369>, rotation=<rot 87.1875 179.9995 -180>}
  {shapeType=Capsule, length=7.5, radius=2, center=<5 0 -1.369>, rotation=<rot 87.1875 179.9995 -180>}
  {shapeType=Capsule, length=7.5, radius=2, center=<5 10 -1.369>, rotation=<rot 87.1875 179.9995 -180>}
 ]
 dyn_roll_01_r [
  {shapeType=Capsule, length=7.5, radius=2, center=<5 -10 1.214>, rotation=<rot 90 -160.5288 -160.5286>}
  {shapeType=Capsule, length=7.5, radius=2, center=<5 0 1.214>, rotation=<rot 90 -179.9999 -179.9998>}
  {shapeType=Capsule, length=7.5, radius=2, center=<5 10 1.214>, rotation=<rot 90 -160.5288 -160.5286>}
 ]
 dyn_roll_02_l [
  {shapeType=Capsule, length=7.5, radius=2, shapeName=dyn_roll_02_l_capsule, center=<5 -10 -0.8256>, rotation=<rot 87.1875 180 -180>}
  {shapeType=Capsule, length=7.5, radius=2, shapeName=dyn_roll_02_l_capsule, center=<5 0 -0.8256>, rotation=<rot 87.1875 180 -180>}
  {shapeType=Capsule, length=7.5, radius=2, shapeName=dyn_roll_02_l_capsule, center=<5 10 -0.8256>, rotation=<rot 87.1875 180 -180>}
 ]
 dyn_roll_02_r [
  {shapeType=Capsule, length=7.5, radius=2, shapeName=dyn_roll_02_l_capsule, center=<5 -10 0.8312>, rotation=<rot 90 172.4476 172.4478>}
  {shapeType=Capsule, length=7.5, radius=2, shapeName=dyn_roll_02_l_capsule, center=<5 0 0.8312>, rotation=<rot 90 172.4476 172.4478>}
  {shapeType=Capsule, length=7.5, radius=2, shapeName=dyn_roll_02_l_capsule, center=<5 10 0.8312>, rotation=<rot 90 172.4476 172.4478>}
 ]
 dyn_roll_MSTR_l [{shapeType=Capsule, length=13, radius=13, shapeName=dyn_roll_MSTR_l_capsule, center=<-2.7195 1.2408 14.2267>, rotation=<rot 90 180 -90>}]
 dyn_roll_MSTR_r [{
   shapeType=Capsule
   length=13
   radius=14
   shapeName=dyn_roll_MSTR_l_capsule
   center=<-2.2681 0.5267 -15.7566>
   rotation=<rot 90 153.4349 -116.5651>
  }]
}
constraints [
 {bone1Name=dyn_roll2_01_l, bone2Name=dyn_roll2_MSTR_l, swing1LimitDegrees=2.5, swing2LimitDegrees=30, twistLimitDegrees=2.5}
 {bone1Name=dyn_roll2_02_l, bone2Name=dyn_roll2_01_l, swing1LimitDegrees=2.5, swing2LimitDegrees=30, twistLimitDegrees=2.5}
 {bone1Name=dyn_roll2_03_l, bone2Name=dyn_roll2_02_l, swing1LimitDegrees=2.5, swing2LimitDegrees=30, twistLimitDegrees=2.5}
 {bone1Name=dyn_roll3_01_l, bone2Name=dyn_roll3_MSTR_l, swing1LimitDegrees=2.5, swing2LimitDegrees=30, twistLimitDegrees=2.5}
 {bone1Name=dyn_roll3_02_l, bone2Name=dyn_roll3_01_l, swing1LimitDegrees=2.5, swing2LimitDegrees=30, twistLimitDegrees=2.5}
 {bone1Name=dyn_roll3_03_l, bone2Name=dyn_roll3_02_l, swing1LimitDegrees=2.5, swing2LimitDegrees=30, twistLimitDegrees=2.5}
 {bone1Name=dyn_roll4_01_l, bone2Name=dyn_roll4_MSTR_l, swing1LimitDegrees=2.5, swing2LimitDegrees=50, twistLimitDegrees=2.5}
 {bone1Name=dyn_roll4_02_l, bone2Name=dyn_roll4_01_l, swing1LimitDegrees=2.5, swing2LimitDegrees=50, twistLimitDegrees=2.5}
 {bone1Name=dyn_roll_01_l, bone2Name=dyn_roll_MSTR_l, swing1LimitDegrees=2.5, swing2LimitDegrees=30, twistLimitDegrees=2.5}
 {bone1Name=dyn_roll_02_l, bone2Name=dyn_roll_01_l, swing1LimitDegrees=2.5, swing2LimitDegrees=30, twistLimitDegrees=2.5}
 {bone1Name=dyn_roll2_01_r, bone2Name=dyn_roll2_MSTR_r, swing1LimitDegrees=2.5, swing2LimitDegrees=50, twistLimitDegrees=2.5}
 {bone1Name=dyn_roll2_02_r, bone2Name=dyn_roll2_01_r, swing1LimitDegrees=2.5, swing2LimitDegrees=50, twistLimitDegrees=2.5}
 {bone1Name=dyn_roll3_01_r, bone2Name=dyn_roll3_MSTR_r, swing1LimitDegrees=2.5, swing2LimitDegrees=50, twistLimitDegrees=2.5}
 {bone1Name=dyn_roll3_02_r, bone2Name=dyn_roll3_01_r, swing1LimitDegrees=2.5, swing2LimitDegrees=50, twistLimitDegrees=2.5}
 {bone1Name=dyn_roll4_01_r, bone2Name=dyn_roll4_MSTR_r, swing1LimitDegrees=2.5, swing2LimitDegrees=50, twistLimitDegrees=2.5}
 {bone1Name=dyn_roll4_02_r, bone2Name=dyn_roll4_01_r, swing1LimitDegrees=2.5, swing2LimitDegrees=50, twistLimitDegrees=2.5}
 {bone1Name=dyn_roll_01_r, bone2Name=dyn_roll_MSTR_r, swing1LimitDegrees=2.5, swing2LimitDegrees=50, twistLimitDegrees=2.5}
 {bone1Name=dyn_roll_02_r, bone2Name=dyn_roll_01_r, swing1LimitDegrees=2.5, swing2LimitDegrees=50, twistLimitDegrees=2.5}
 {
  bone1Name=dyn_roll_MSTR_l
  bone2Name=Melee_TwoHanded_Handle
  swing1LimitDegrees=45
  swing1Motion=Free
  swing2LimitDegrees=45
  swing2Motion=Free
  twistLimitDegrees=45
  twistMotion=Free
 }
 {
  bone1Name=dyn_roll2_MSTR_l
  bone2Name=Melee_TwoHanded_Handle
  swing1LimitDegrees=45
  swing1Motion=Free
  swing2LimitDegrees=45
  swing2Motion=Free
  twistLimitDegrees=45
  twistMotion=Free
 }
 {
  bone1Name=dyn_roll3_MSTR_l
  bone2Name=Melee_TwoHanded_Handle
  swing1LimitDegrees=45
  swing1Motion=Free
  swing2LimitDegrees=45
  swing2Motion=Free
  twistLimitDegrees=45
  twistMotion=Free
 }
 {
  bone1Name=dyn_roll4_MSTR_l
  bone2Name=Melee_TwoHanded_Handle
  swing1LimitDegrees=45
  swing1Motion=Free
  swing2LimitDegrees=45
  swing2Motion=Free
  twistLimitDegrees=45
  twistMotion=Free
 }
 {
  bone1Name=dyn_roll_MSTR_r
  bone2Name=Melee_TwoHanded_Handle
  swing1LimitDegrees=45
  swing1Motion=Free
  swing2LimitDegrees=45
  swing2Motion=Free
  twistLimitDegrees=45
  twistMotion=Free
 }
 {
  bone1Name=dyn_roll2_MSTR_r
  bone2Name=Melee_TwoHanded_Handle
  swing1LimitDegrees=45
  swing1Motion=Free
  swing2LimitDegrees=45
  swing2Motion=Free
  twistLimitDegrees=45
  twistMotion=Free
 }
 {
  bone1Name=dyn_roll3_MSTR_r
  bone2Name=Melee_TwoHanded_Handle
  swing1LimitDegrees=45
  swing1Motion=Free
  swing2LimitDegrees=45
  swing2Motion=Free
  twistLimitDegrees=45
  twistMotion=Free
 }
 {
  bone1Name=dyn_roll4_MSTR_r
  bone2Name=Melee_TwoHanded_Handle
  swing1LimitDegrees=45
  swing1Motion=Free
  swing2LimitDegrees=45
  swing2Motion=Free
  twistLimitDegrees=45
  twistMotion=Free
 }
 {
  bone1Name=dyn_handle_paper_MSTR
  bone2Name=Melee_TwoHanded_Handle
  swing1LimitDegrees=45
  swing1Motion=Locked
  swing2LimitDegrees=45
  swing2Motion=Locked
  twistLimitDegrees=45
  twistMotion=Locked
 }
 {bone1Name=dyn_handle_paper_01, bone2Name=dyn_handle_paper_MSTR, swing1LimitDegrees=20, swing2LimitDegrees=50, twistLimitDegrees=20}
]
